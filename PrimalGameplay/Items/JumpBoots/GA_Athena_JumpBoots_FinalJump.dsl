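; Gameplay ability for the final Jump Boots jump, starting a skydive when the player reaches jump apex.
; Activation plays the launch pose montage, binds the apex event, and sets up end conditions.
; Landing or gaining a swimming, vehicle or zipline tag ends the ability so it cannot get stuck on.
; Ending unbinds the apex event in case of cancellation and removes the granting effect with its GameplayCue.
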(event Ability|EventActivateAbility
  (bind _outfortplayerpawn (CallFunction|GetAvatarasValidFortPlayerPawn)
    (:ValidFortPlayerPawn
      (Variables|Default|SetPlayerRef _outfortplayerpawn)
      (Character|BindEventtoOnReachedJumpApex (Variables|Default|GetPlayerRef) (EventDispatchers|CreateEvent :!event Custom|BeginSkydive)) ; Begin skydive at jump apex
      (CallFunction|BindtoEndEvents) ; Bind to events that will end the ability
      (Ability|Tasks|PlayMontageAndWaitNotify "None" "/Game/Athena/Items/Gameplay/JumpBoots/Assets/Montage_JumpBoots_FinalJump.Montage_JumpBoots_FinalJump") ; Play superman pose while launched upwards!
    )))

(event Custom|BeginSkydive
  (Character|UnbindEventfromOnReachedJumpApex (Variables|Default|GetPlayerRef) (delegate Custom|BeginSkydive))
  (Parachute|BeginSkydiving (Variables|Default|GetPlayerRef) false))

(event Ability|EventOnEndAbility
  (params (bWasCancelled :type Boolean))
  (|CallParentFunction bWasCancelled)
  (Character|UnbindEventfromOnReachedJumpApex (Variables|Default|GetPlayerRef) (EventDispatchers|CreateEvent :!event Custom|BeginSkydive)) ; Unbind Skydive event if case GA is cancelled
  (Ability|RemoveGrantedbyEffect) ; FORT-351341\n\nRemove GE and associated GCN.
)

(event Custom|BindToEndEvents
  ;[1 1 1] Bind to things that should end this ability such as landing, entering water, or a vehicle. So that the player does not get stuck with the ability on.
  ;;[1 1 1] Set up Land Event
  (EventDispatchers|BindEventtoOnPawnLanded (Variables|Default|GetPlayerRef) (delegate Custom|Landed))
  (Ability|Tasks|WaitGameplayTagAdd (Variables|Default|GetTag_SurfaceSwimming)
    (:then
      (Ability|Tasks|WaitGameplayTagAdd (Variables|Default|GetTag_InVehicle)
        (:then
          (Ability|Tasks|WaitGameplayTagAdd (Variables|Default|GetTag_Zipline)
            (:Added
              _)))
        (:Added
          _))
      (Ability|EndAbility))
    (:Added
      _)))

(event Custom|Landed
  (params (Hit :type Hit%20Result%20Structure%20%28by%20ref%29))
  ;[1 1 1] Bind to things that should end this ability such as landing, entering water, or a vehicle. So that the player does not get stuck with the ability on.
  (Ability|EndAbility))
